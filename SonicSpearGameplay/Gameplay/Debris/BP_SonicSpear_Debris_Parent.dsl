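; Parent debris actor for SonicSpear donut drops that leaves a frosting splat decal, crumbs and frosting clumps, tinted by a server-set color preset.
; Cosmetic-capable instances fill box bounds with ISM debris using landscape traces, slope checks, distance-based scaling and platform-dependent counts; the construction script previews it in editor.
; Unless cleanup is disabled, the server can follow parent death while clients time material-driven hiding by detail mode, skipping Epic.
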(event EventBeginPlay
  (bind (_colordatatable_datatable _colordatatable_rowname) (Utilities|Struct|BreakDataTableRowHandle (Variables|GetColorDataTable_DataTable)))
  (|CallParentFunction)
  (Utilities|FlowControl|Sequence
    (:then_0
      ;[0 0.02781 0.15 0.5] initialize frosting color on the donuts
      (if (Variables|Config|Color|GetbUseColorPreset)
        (Networking|SwitchHasAuthority
          (:Authority
            (bind _returnvalue (Utilities|GetDataTableRow _colordatatable_datatable _colordatatable_rowname))
            (Variables|Config|Color|SetColoring _returnvalue))) ; set the color from server
      ))
    (:then_1
      (Networking|CanExecuteCosmeticEvents
        (:True
          (CallFunction|FormatComponents)
          (CallFunction|SpawnDebris))))
    (:then_2
      (if (not (Variables|Default|GetbNeverCleanup))
        (Networking|SwitchHasAuthority
          (:Authority
            ;[1 1 1] Server will track if the parent has died and will remove it for everyone if so
            (if (Variables|Perf|GetCleanupDebrisAfterParentDestroyed)
              (bind _sonicspearparent (Variables|Runtime|GetSonicSpearParent)
                (:then
                  (EventDispatchers|BindEventtoOnDeathPlayEffectsDelegate _sonicspearparent (EventDispatchers|CreateEvent :!event Custom|DestroyDebris)))
                (:else))))
          (:Remote
            ;[1 1 1] Cleanup time Dependant on Scalability per Client settings
            ; No Cleanup on Epic
            (if (< (Rendering|GetRenderingDetailMode) 3)
              (bind _returnvalue_2 (Utilities|Time|SetTimerbyEvent (EventDispatchers|CreateEvent :!event Custom|PerScalabilityDespawn) (CallFunction|GetScalableAutoDestroyTime)))
              (Variables|Default|SetAutoDestroyTimer _returnvalue_2))))))))

(event Custom|DespawnAndDestroyAllMulti
  ;[1 1 1] Start shrinking - then kill when done
  (CallFunction|HideDebris true) ; Low
  (Networking|SwitchHasAuthority
    (:Authority
      (Actor|SetLifeSpan 5.0) ; Need to wait until everything fades out
    )))

(event Custom|DestroyDebris
  (params (Damage :type Float%20%28single-precision%29) (DamageTags :type Gameplay%20Tag%20Container%20Structure%20%28by%20ref%29) (Momentum :type Vector) (HitInfo :type Hit%20Result%20Structure%20%28by%20ref%29) (InstigatedBy :type Fort%20Pawn%20Object%20Reference) (DamageCauser :type Actor%20Object%20Reference) (EffectContext :type Gameplay%20Effect%20Context%20Handle%20Structure))
  (Utilities|Time|ClearandInvalidateTimerbyHandle (Variables|Default|GetAutoDestroyTimer))
  (CallFunction|DespawnandDestroyAllMulti))

(event Custom|PerScalabilityDespawn
  (Networking|CanExecuteCosmeticEvents
    (:True
      ;[1 1 1] Epic = permanent decal and crumbs
      ; high = longer tail decal and permanent crumbs
      ; medium = timed decal and permanent crumbs
      ; low = shorter timed decal and crumbs
      (switch int (Rendering|GetRenderingDetailMode)
        (:0
          ;[1 1 1] Start shrinking - then kill when done
          (CallFunction|HideDebris true) ; Low
          (Networking|SwitchHasAuthority
            (:Authority
              (Actor|SetLifeSpan 5.0) ; Need to wait until everything fades out
            )))
        (:1
          (CallFunction|HideDebris false) ; Medium
        )
        (:2
          (CallFunction|HideDebris false) ; High
        ))
      ;;[1 1 1] Remove Decal Only
      (Utilities|FlowControl|Delay 5.0)
      (Rendering|SetVisibility (Variables|Default|GetFrostingDecal) false false)
      (Components|DestroyComponent (Variables|Default|GetFrostingDecal)))))

(event Custom|SpawnDebris
  (Utilities|FlowControl|Sequence
    (:then_0
      ;[1 1 1] Decal Setup
      (if (and (Variables|Config|Decal|GetbUseSplatDecal) (Utilities|Select :Index (Variables|/Script/Engine.SceneComponent|GetDetailMode (Variables|Default|GetFrostingDecal)) :DM_Low true :DM_Medium true :DM_High false :DM_Epic false))
        (bind (_decalref _valid) (Decal|FindandSetDecal))
        (if _valid
          (bind _returnvalue (Rendering|Material|CreateDynamicMaterialInstance (Variables|Config|Decal|GetDecalMaterialParent) "None" "None" :!class /Script/Engine.KismetMaterialLibrary)) ; The Decal Material INSTANCE, needs to be the parent mat here
          (bind _output_get (Variables|Runtime|SetDecalDynamicMaterialInstance _returnvalue))
          (Rendering|Components|Decal|SetDecalMaterial (Variables|Default|GetDecal) _output_get)
          (Rendering|Material|SetVectorParameterValue (Variables|Runtime|GetDecalDynamicMaterialInstance) "Frosting Color" (Utilities|Struct|BreakSSonicSpearDonutDropColor (Variables|Config|Color|GetColoring)) :!class /Script/Engine.MaterialInstanceDynamic)
          (Rendering|Material|SetScalarParameterValue (Variables|Runtime|GetDecalDynamicMaterialInstance) "Shrinking Enabled" 0.0 :!class /Script/Engine.MaterialInstanceDynamic)
          (CallFunction|SetDecalScale))
        (else
          (Components|DestroyComponent (Variables|Default|GetFrostingDecal)))))
    (:then_1
      (if (not (Variables|Default|GetbNeverCleanup))
        ;[1 1 1 1 12] By waiting just long enough for other actors to completely destroy from the impact, there is less of a chance for crumbs to land on something that is about to be destroyed causing the floating crumbs
        (Utilities|FlowControl|Delay 0.15)
        ;[1 1 1] Frosting Clumps
        (Crumbs|SpawnFrostingClumpsISM)
        ;[1 1 1] Minor Optimization to spread spawning out
        (Utilities|FlowControl|DelayUntilNextFrame)
        ;[1 1 1] Crumbs
        (Crumbs|SpawnCrumbsISM)
        (else
          (Crumbs|SpawnFrostingClumpsISM)
          (Crumbs|SpawnCrumbsISM))))))

(fn ColorCrumbCustomPrimISM
  (params (ISM :type Instanced%20Static%20Mesh%20Component%20Object%20Reference))
  (Rendering|Material|SetCustomPrimitiveDataVector4 ISM 2 (Class|CreativeColorComponentBlueprintLibrary|RGBAToVec4 (Utilities|Struct|BreakSSonicSpearDonutDropColor (Variables|Config|Color|GetColoring)))))

(fn FindAndSetDecal
  (bind _decalholder (Variables|Default|GetDecalHolder))
  (bind _children (Components|GetChildrenComponents _decalholder false))
  (bind _output (Utilities|Array|Get(acopy) _children))
  (if (Utilities|Array|IsNotEmpty _children)
    (Utilities|IsValid _output
      (:Is%20Valid
        (bind _asdecal_component (Utilities|Casting|CastToDecalComponent _output)
          (:then
            (bind _output_get (Variables|Default|SetDecal _asdecal_component))
            (return _output_get true))
          (:CastFailed
            (return 0 false))))
      (:Is%20Not%20Valid
        _)))
  (return 0 false))

(fn FormatComponents
  ;[1 1 1] Update Components
  (Components|Box|SetBoxExtent (Variables|Default|GetFrostingClumpsBounds) (Variables|Config|GetFrostingClumpsBoxExtents) false)
  (Components|Box|SetBoxExtent (Variables|Default|GetCrumbBounds) (Variables|Config|GetCrumbBoxExtents) false)
  ;;[1 1 1] We set the bounds up higher than the impact point in for cases when we are on a hill or odd angle.
  ;; Allowing the Crumbs to start tracing for the landscape at a higher distance/on an incline
  (Transformation|SetRelativeLocation (Variables|Default|GetCrumbBounds) (Math|Vector|MakeVector 0.0 (Variables|Config|GetCrumbBoundsOffset) 3e+02) false false)
  (Transformation|SetRelativeLocation (Variables|Default|GetFrostingClumpsBounds) (Math|Vector|MakeVector 0.0 (Variables|Config|GetFrostingBoundsOffset) 3e+02) false false))

(fn GenerateTransformsInBounds
  (params (Amount :type Integer) (TargetBounds :type Box%20Collision%20Object%20Reference) (BaseScale :type Float%20%28double-precision%29))
  (bind _returnvalue (Transformation|GetActorRotation))
  (bind _targetbounds (Variables|GetTargetBounds))
  (bind (_returnvalue_x _returnvalue_y) (Math|Vector|BreakVector (Components|Box|GetScaledBoxExtent _targetbounds)))
  (bind (_bblockinghit _binitialoverlap _time _distance _location _impactpoint _normal _impactnormal) (Collision|BreakHitResult (LocalVariables|GetHitResult)))
  (bind _hitresult (LocalVariables|GetHitResult))
  (bind (_bblockinghit_2 _binitialoverlap_2 _time_2 _distance_2 _location_2 _impactpoint_2 _normal_2 _impactnormal_2) (Collision|BreakHitResult _hitresult))
  (bind _relativelocation (Variables|/Script/Engine.SceneComponent|GetRelativeLocation _targetbounds))
  (bind _returnvalue_2 (Math|Random|RandomPointinBoundingBox _relativelocation (Math|Vector|MakeVector _returnvalue_x _returnvalue_y 1.0)))
  (bind _yaw (.yaw _returnvalue))
  (bind _returnvalue_3 (Math|Vector|RotateVectorAroundAxis _returnvalue_2 _yaw "0.000000,0.000000,1.000000"))
  (bind (_x _y _z) (Math|Rotator|BreakRotIntoAxes (Math|Rotator|MakeRotfromZ _impactnormal)))
  (bind (_boxextent_x _boxextent_y) (Math|Vector|BreakVector (Variables|/Script/Engine.BoxComponent|GetBoxExtent_X (Variables|GetTargetBounds))))
  (Utilities|Array|Clear (LocalVariables|GetTransforms))
  (for _index (range Amount)
    (LocalVariables|SetIndex _index)
    (if (Variables|Config|Advanced|GetbDoTrace)
      (bind _output_get (LocalVariables|SetTraceStart (+ (Transformation|GetActorLocation) _returnvalue_3)))
      (bind (_outhit _returnvalue_4) (Collision|LineTraceByProfile _output_get (Utilities|Operators|Add _output_get (Math|Vector|MakeVector 0.0 0.0 (+ (Variables|Config|Advanced|GetTraceDistance) (.z _returnvalue_3)))) "FortLandscapeAndWater" false (Utilities|Array|MakeArray (Actor|GetOwner) (Variables|Runtime|GetSonicSpearParent)) (Variables|Perf|GetDebugCrumpTrace) true "(R=1.000000,G=0.000000,B=0.000000,A=1.000000)" "(R=0.000000,G=1.000000,B=0.000000,A=1.000000)" 0.1)) ; Trace for Landscape
      (if _returnvalue_4
        (LocalVariables|SetHitResult _outhit)
        (Utilities|IsA(soft) (.HitActor (Collision|BreakHitResult _outhit)) "/SonicSpearGameplay/Gameplay/Drops/BP_SonicSpear_Donut.BP_SonicSpear_Donut_C"
          (:False
            (Utilities|IsA(soft) (.HitActor (Collision|BreakHitResult _outhit)) "/SonicSpearGameplay/Gameplay/Drops/B_SonicSpear_Drop_Physics_Parent.B_SonicSpear_Drop_Physics_Parent_C"
              (:False
                (Utilities|IsA(soft) (.HitActor (Collision|BreakHitResult _outhit)) "/SonicSpearGameplay/Gameplay/Drops/BGA_SonicSpear_MainDropParent.BGA_SonicSpear_MainDropParent_C"
                  (:False
                    _)) ; Dont spawn on a drop parent itself
              )) ; Dont spawn crumbs on physics donuts
          )) ; Dont spawn crumbs on the donut
      )
      (else
        _))
    ;[1 1 1] Configure Transform Result
    (LocalVariables|SetRandScale (Math|Random|RandomFloatinRange (Variables|Config|Crumbs|GetCrumbMinScale) (Variables|Config|Crumbs|GetCrumbMaxScale)))
    (if (>= (Math|Vector|DotProduct _impactnormal_2 "0.000000,0.000000,1.000000") (Variables|Config|Advanced|GetMinCrumbSlope))
      (Utilities|Array|Add (LocalVariables|GetTransforms) (Math|Transform|MakeTransform (Utilities|Operators|Add (select (Variables|Config|Advanced|GetbDoTrace) _location_2 (LocalVariables|GetTraceStart)) (Math|Vector|MakeVector 0.0 0.0 (Variables|Config|Crumbs|GetCrumbZOffset))) (Math|Rotator|MakeRotfromZX _z (Math|Vector|RotateVectorAroundAxis "1.000000,0.000000,0.000000" (Math|Random|RandomFloatinRange 0.0 359.0) _z)) (Utilities|Operators|Multiply (Variables|GetBaseScale) (Math|Conversions|ToVector(Float) (LocalVariables|GetRandScale)) (Math|Interpolation|Ease "EaseOut" (Math|Float|Clamp(Float) (/ (Math|Vector|Distance2D(Vector) (select (Variables|Config|Advanced|GetbDoTrace) _location_2 (LocalVariables|GetTraceStart)) (Transformation|GetActorLocation)) (Math|Float|Max(Float) _boxextent_x _boxextent_y))) 1.0 (Variables|Config|Advanced|GetDistanceMinScale) (Variables|Config|Advanced|GetDistanceExpoBlend)))))))
  (return (LocalVariables|GetTransforms)))

(fn GetScalableAutoDestroyTime
  (return (Utilities|Select :Index (Rendering|GetRenderingDetailMode) :Option%200 (Variables|Default|GetAutoDestroyTime_Low) :Option%201 (Variables|Default|GetAutoDestroyTime_Med) :Option%202 (Variables|Default|GetAutoDestroyTime_High) :Option%203 6e+02)))

(fn HideDebris
  (params (HideCrumbs :type Boolean :default true) (HideDecal :type Boolean :default true))
  (bind _decaldynamicmaterialinstance (Variables|Runtime|GetDecalDynamicMaterialInstance))
  (Utilities|FlowControl|Sequence
    (:then_0
      (if (Variables|GetHideCrumbs)
        (Rendering|Material|SetCustomPrimitiveDataFloat (Variables|Default|GetISM_Crumbs) 0 (Utilities|Time|GetGameTimeinSeconds))
        (Rendering|Material|SetCustomPrimitiveDataFloat (Variables|Default|GetISM_FrostingClumps) 1 1.0)))
    (:then_1
      (if (and (Variables|Config|Decal|GetbUseSplatDecal) (Variables|GetHideDecal))
        (Rendering|Material|SetScalarParameterValue _decaldynamicmaterialinstance "CurrentGameTime" (Utilities|Time|GetGameTimeinSeconds) :!class /Script/Engine.MaterialInstanceDynamic)
        (Rendering|Material|SetScalarParameterValue _decaldynamicmaterialinstance "Shrinking Enabled" 1.0 :!class /Script/Engine.MaterialInstanceDynamic)))))

(fn OnRep_SizeMulti
  (CallFunction|SetDecalScale))

(fn SetDecalScale
  (bind (_relativescale3d_x _relativescale3d_y _relativescale3d_z) (Math|Vector|BreakVector (Variables|/Script/Engine.SceneComponent|GetRelativeScale3D_X (Variables|Default|GetFrostingDecal))))
  ;[1 1 1] Scale the details with DOnut scalar
  (Transformation|SetRelativeScale3D (Variables|Default|GetFrostingDecal) (Math|Vector|MakeVector (* _relativescale3d_x (Variables|Config|Decal|GetDecalScale)) (* (* _relativescale3d_y (Variables|Runtime|GetSizeMulti)) (Variables|Config|Decal|GetDecalScale)) (* (* _relativescale3d_z (Variables|Runtime|GetSizeMulti)) (Variables|Config|Decal|GetDecalScale)))))

(fn SpawnCrumbsISM
  (bind _returnvalue (TouchPlatform|IsMobileGame))
  (bind _returnvalue_2 (TouchPlatform|IsNintendoSwitch))
  (bind _returnvalue_3 (or _returnvalue _returnvalue_2))
  (bind _numberofcrumbshigh (Variables|Config|Advanced|GetNumberOfCrumbsHigh))
  (bind _numberofcrumbslow (Variables|Config|Advanced|GetNumberOfCrumbsLow))
  (bind _returnvalue_4 (Math|Integer|SelectInt _numberofcrumbslow _numberofcrumbshigh _returnvalue_3))
  (Components|InstancedStaticMesh|ClearInstances (Variables|Default|GetISM_Crumbs))
  (if (> _returnvalue_4 0)
    (bind _ismtransforms (Crumbs|GenerateTransformsinBounds _returnvalue_4 (Variables|Default|GetCrumbBounds) (Variables|Config|Crumbs|GetCrumbBaseScale)))
    (Components|InstancedStaticMesh|AddInstances (Variables|Default|GetISM_Crumbs) _ismtransforms false true false)
    (Crumbs|ColorCrumbCustomPrimISM (Variables|Default|GetISM_Crumbs))))

(fn SpawnFrostingClumpsISM
  (bind _returnvalue (TouchPlatform|IsMobileGame))
  (bind _returnvalue_2 (TouchPlatform|IsNintendoSwitch))
  (bind _returnvalue_3 (or _returnvalue _returnvalue_2))
  (bind _numberoffrostingclumpshigh (Variables|Config|Advanced|GetNumberOfFrostingClumpsHigh))
  (bind _numberoffrostingclumpslow (Variables|Config|Advanced|GetNumberOfFrostingClumpsLow))
  (bind _returnvalue_4 (Math|Integer|SelectInt _numberoffrostingclumpslow _numberoffrostingclumpshigh _returnvalue_3))
  (Components|InstancedStaticMesh|ClearInstances (Variables|Default|GetISM_FrostingClumps))
  (if (> _returnvalue_4 0)
    (bind _ismtransforms (Crumbs|GenerateTransformsinBounds _returnvalue_4 (Variables|Default|GetFrostingClumpsBounds) (Variables|Config|Frosting|GetFrostingClumpsBaseScale)))
    (Components|InstancedStaticMesh|AddInstances (Variables|Default|GetISM_FrostingClumps) _ismtransforms false true false)
    (Crumbs|ColorCrumbCustomPrimISM (Variables|Default|GetISM_FrostingClumps))))

(fn ConstructionScript
  (bind (_colordatatable_datatable _colordatatable_rowname) (Utilities|Struct|BreakDataTableRowHandle (Variables|GetColorDataTable_DataTable)))
  ;[0.15 0.1097 0 0.5] show proper frosting colors for donuts pre-placed in levels (in editor only)
  (if (Editor|InEditor)
    (if (Variables|Config|Color|GetbUseColorPreset)
      (bind _returnvalue (Utilities|GetDataTableRow _colordatatable_datatable _colordatatable_rowname))
      (Variables|Config|Color|SetColoring _returnvalue)
      (else
        _))
    (Networking|CanExecuteCosmeticEvents
      (:True
        (CallFunction|FormatComponents)
        (CallFunction|SpawnDebris)))))
